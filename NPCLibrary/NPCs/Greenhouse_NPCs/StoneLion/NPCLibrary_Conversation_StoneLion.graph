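# Conversation database for the StoneLion NPC, who introduces herself as Elizabeth, defining her greeting and the branches that follow it.
# A single entry point leads to a Speech node whose default line gives way to priority-curve contextual lines for first-ever conversations and three specific player outfits.
# The greeting fans out through link nodes to remote entry tags for NPC library quests, basic services, StoneLion dialogue and an S30 Elizabeth conversation.
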ConversationDatabase /NPCLibrary/NPCs/Greenhouse_NPCs/StoneLion/NPCLibrary_Conversation_StoneLion
entry_points [{tag=Conversation.Entry.StoneLion, destinations [#1]}]
nodes [
 ConversationEntryPointNode_7 {
  id=#1
  class=ConversationEntryPointNode
  outputs [#2]
  properties {entryTag {tagName=Conversation.Entry.StoneLion}}
 }
 Speech_C_4 {
  id=#2
  class=Speech_C
  outputs [#3, #4, #5, #6]
  properties {
   nodeName=""
   generalConfig {
    defaultMessage="Please, call me Elizabeth."
    contextualMessages [
     {
      message="Hoist the colors!"
      contextRequirements [{requirement {class=FortControllerRequirement_IsFirstEverConversationWithNPC}}]
      priority {curve {rowName=Tandem.Speech.Priority.FirstEverInteraction}}
     }
     {
      message="I hardly believe in ghost stories anymore, Captain Barbossa."
      contextRequirements [{requirement {class=FortControllerRequirement_HasCID, properties {allowedCIds [/BRCosmetics/Athena/Items/Cosmetics/Characters/Character_DomeRoof]}}}]
      priority {curve {rowName=Tandem.Speech.Priority.OutfitReaction}}
     }
     {
      message="You have betrayed your duty, Davy Jones."
      contextRequirements [{requirement {class=FortControllerRequirement_HasCID, properties {allowedCIds [/BRCosmetics/Athena/Items/Cosmetics/Characters/Character_AccentWall]}}}]
      priority {curve {rowName=Tandem.Speech.Priority.OutfitReaction}}
     }
     {
      message=There will come a moment when you have the chance to do the right thing.
      contextRequirements [{requirement {class=FortControllerRequirement_HasCID, properties {allowedCIds [/BRCosmetics/Athena/Items/Cosmetics/Characters/Character_SpringBreak]}}}]
      priority {curve {rowName=Tandem.Speech.Priority.OutfitReaction}}
     }
    ]
   }
  }
 }
 ConversationLinkNode_3 {
  id=#6
  class=ConversationLinkNode
  properties {remoteEntryTag {tagName=Conversation.Entry.S30.D3.03.Elizabeth}}
 }
 ConversationLinkNode_2 {id=#5, class=ConversationLinkNode, properties {remoteEntryTag {tagName=Conversation.Entry.StoneLion.Dialogue}}}
 ConversationLinkNode_0 {id=#4, class=ConversationLinkNode, properties {remoteEntryTag {tagName=Conversation.Entry.BasicServices}}}
 ConversationLinkNode_1 {id=#3, class=ConversationLinkNode, properties {remoteEntryTag {tagName=Conversation.Entry.NPCLibrary.Quests}}}
]
